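# Unlit, translucent, two-sided material for Niagara ribbons in the BraveBuild pickaxe VFX, drawing three red streaks.
# Procedural UV math builds three soft bands across V, adds a solid cap near U=0 and fades along U, without textures.
# The squared mask blends dark to bright red through an emissive-adjustments function, while the mask times particle alpha gives a fade value.
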Material /BRCosmetics/Weapons/FORT_Melee/Pickaxe_BraveBuild/VFX/M_Pickaxe_BraveBuild_3Lines
graph {
 properties {
  bUsedWithNiagaraRibbons=true
  blendMode=BLEND_Translucent
  pixelDepthOffsetMode=PDOM_Legacy
  shadingModel=MSM_Unlit
  twoSided=true
  usedShadingModels=Unlit
 }
 nodes [
  n1 = TextureCoordinate()
  n2 = ComponentMask(None=@n1) {g=true}
  n3 = Multiply(A=@n2) {constB=3}
  n4 = Multiply(A=OneMinus(None=@n2)) {constB=3}
  n5 = Multiply(A=Multiply(A=Frac(None=@n3), B=Frac(None=@n4))) {constB=2}
  n6 = Multiply(A=SmoothStep(Value=Saturate(None=@n5)) {constMax=0.8, constMin=0.25}) {constB=5}
  n7 = OneMinus(None=ComponentMask(None=@n1) {r=true})
  n8 = SmoothStep(Value=@n7) {constMax=0.95, constMin=0.8}
  n9 = Multiply(A=Saturate(None=Add(A=@n6, B=@n8)), B=@n7)
  MaterialFunctionCall(Emissive=LinearInterpolate(
    A=Constant3Vector() {constant=<rgba 0.099 0 0.0025 1>}
    B=Constant3Vector() {constant=<rgba 0.8802 0.0121 0 1>}
    Alpha=Multiply(A=@n9, B=@n9)
   )) {materialFunction=&MF_FX_EmissiveAdjustments}
  Multiply(A=@n9, B=ParticleColor().A)
 ]
}
