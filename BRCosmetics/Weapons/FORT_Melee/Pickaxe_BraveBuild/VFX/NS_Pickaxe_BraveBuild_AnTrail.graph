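# Niagara system for the BraveBuild pickaxe's animation trail, drawing short-lived ribbons that follow bones or sockets on the user-supplied skeletal mesh.
# Two emitters each spawn two trails on separate bone index pairs: one a dim neutral streak, the other a briefer red-pink streak with a bright start.
# Both fade and taper over life through curves.
# Time of day tints the first and dims the second in evening and night.
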NiagaraSystem /BRCosmetics/Weapons/FORT_Melee/Pickaxe_BraveBuild/VFX/NS_Pickaxe_BraveBuild_AnTrail
dependencies {
 usedDataInterfaces [&NiagaraDataInterfaceColorCurve, &NiagaraDataInterfaceCurve]
 usedDynamicInputs []
 usedModules [&SystemState]
}
emitters {
 FN_AnimTrail {
  emitterSpawnScript [&FN_RibbonIDOffsetUtility()]
  emitterUpdateScript [
   &EmitterState()
   &FN_SpawnAnimTrail(Number of Anim Trails=2)
  ]
  particleSpawnScript [
   &FN_InitializeAnimTrailBoneIndecies()
   &FN_InitializeAnimTrail(AnimTrailWidth=1.5, Color=<rgba 1 1 1 0.55>, Lifetime=0.15)
  ]
  particleUpdateScript [
   &FN_UpdateAnimTrail(Locked Particle RibbonWidth:NiagaraFloat=$Particles.RibbonWidth, SkeletalMesh=$User.SkeletalMesh) {module=FN_UpdateAnimTrail001}
   &ScaleColor(
    Linear Color Curve=DataInterface() {propertyValues {
      AlphaCurve {keys [<at 0 0.7>,<at 0.038 1>,<at 0.2257 1>,<at 0.6771 0.3>,<at 0.9344 0.05>,<at 1 0>]}
      BlueCurve {keys [<at 0 0.1562>,<at 0.5969 0.0938>]}
      GreenCurve {keys [<at 0 0.1562>,<at 0.5969 0.0938>]}
      RedCurve {keys [<at 0 0.1562>,<at 0.5969 0.0938>]}
     }}
    Scale Mode=::NewEnumerator2
   )
   &Color(Color=&MultiplyLinearColor(
     Linear Color=$Particles.Color
     Scale Factor=&MakeVector4_FromLinearColor(LinearColor=&FN_Select_Color_from_Time_Of_Day(Morning=<rgba 0.6719 0.361 0.5226 1>, Day=<rgba 1 1 1 1>, Evening=<rgba 1 0.6413 0.7956 1>, Night=<rgba 0.3188 0.3311 0.5885 1>))
    )) {module=Color001}
   &ScaleRibbonWidth(Ribbon Width Scale=&FloatFromCurve(FloatCurve=DataInterface() {propertyValues {Curve {keys [
         {interpMode=RCIM_Cubic, tangentMode=RCTM_User, time=0, value=1}
         {arriveTangent=-3.1724, interpMode=RCIM_Cubic, leaveTangent=-3.1724, tangentMode=RCTM_User, time=1, value=0}
        ]}}}))
   &SolveForcesAndVelocity(Clamp Velocity=false, Limit Acceleration=false)
  ]
  rendererClasses [&NiagaraRibbonRendererProperties]
  renderers [{rendererClass=&NiagaraRibbonRendererProperties}]
 }
 FN_AnimTrail001 {
  emitterSpawnScript [&FN_RibbonIDOffsetUtility()]
  emitterUpdateScript [
   &EmitterState()
   &FN_SpawnAnimTrail(Number of Anim Trails=2)
  ]
  particleSpawnScript [
   &FN_InitializeAnimTrailBoneIndecies(First Bone or Socket Index=2, Second Bone or Socket Index=3)
   &FN_InitializeAnimTrail(AnimTrailWidth=1.4, Color=<rgba 0.6927 0 0.0184 0.5>, Lifetime=0.1)
  ]
  particleUpdateScript [
   &FN_UpdateAnimTrail(Locked Particle RibbonWidth:NiagaraFloat=$Particles.RibbonWidth, SkeletalMesh=$User.SkeletalMesh) {module=FN_UpdateAnimTrail001}
   &ScaleColor(
    Linear Color Curve=DataInterface() {propertyValues {
      AlphaCurve {keys [<at 0 0.3>,<at 0.0169 1>,<at 0.232 1>,<at 0.4746 0.12>,<at 1 0>]}
      BlueCurve {keys [<at 0 1.5>,<at 0.1498 1>,<at 1 0.0031>]}
      GreenCurve {keys [<at 0 1.5>,<at 0.1498 1>,<at 1 0>]}
      RedCurve {keys [<at 0 1.5>,<at 0.1498 1>,<at 1 0.1406>]}
     }}
    Scale Mode=::NewEnumerator2
   )
   &Color(Color=&MultiplyLinearColor(
     Linear Color=$Particles.Color
     Scale Factor=&MakeVector4_FromFloat(Float=&FN_Select_Float_from_Time_Of_Day(Morning=1, Day=1, Evening=0.65, Night=0.6))
    )) {module=Color001}
   &ScaleRibbonWidth(Ribbon Width Scale=&FloatFromCurve(FloatCurve=DataInterface() {propertyValues {Curve {keys [
         {arriveTangent=-0.0084, interpMode=RCIM_Cubic, leaveTangent=-0.0084, tangentMode=RCTM_User, time=0, value=1}
         {arriveTangent=-2.9376, interpMode=RCIM_Cubic, leaveTangent=-2.9376, tangentMode=RCTM_User, time=1, value=0}
        ]}}}))
   &SolveForcesAndVelocity(Clamp Velocity=false, Limit Acceleration=false)
  ]
  rendererClasses [&NiagaraRibbonRendererProperties]
  renderers [{rendererClass=&NiagaraRibbonRendererProperties}]
 }
}
summary {
 systemName=NS_Pickaxe_BraveBuild_AnTrail
 emitters [
  {emitterName=FN_AnimTrail, rendererClasses [&NiagaraRibbonRendererProperties]}
  {emitterName=FN_AnimTrail001, rendererClasses [&NiagaraRibbonRendererProperties]}
 ]
 userVariables [
  User.RibbonIDOffset {type=NiagaraInt32}
  User.SkeletalMesh {
   type=NiagaraDataInterfaceSkeletalMesh
   defaultValue {
    struct=NiagaraExt_VariableValue_DataInterface
    value {dataInterface=&self:NiagaraDataInterfaceSkeletalMesh_0}
   }
  }
 ]
}
